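# Physics asset for the F_MED Winter Holiday Hat cosmetic, defining simple collision and joint limits for its dyn_hat_01 bone.
# A capsule body covers dyn_hat_01, and an offset, tilted box body stands in for the head.
# A single constraint links the hat bone to the head, with asymmetric swing limits and an almost zero twist limit, so the hat section can bend but barely twist.
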PhysicsAsset /BRCosmetics/Accessories/Hats/F_MED_Winter_Holiday_Hat/Meshes/F_MED_Winter_Holiday_Physics
bodies {
 dyn_hat_01 [{shapeType=Capsule, length=9.3257, radius=2.5, shapeName=dyn_hat_01_capsule, center=<6.3298 0 0>, rotation=<rot 90 0 0>}]
 head [{shapeType=Box, extentX=9.3941, extentY=17.4231, extentZ=18, center=<3.2137 -5.7055 -2.2444>, rotation=<rot -5 -40.1251 0>}]
}
constraints [{bone1Name=dyn_hat_01, bone2Name=head, swing1LimitDegrees=45, swing2LimitDegrees=20, twistLimitDegrees=0.2}]
