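# Animation Blueprint for the F_MED winter holiday hat cosmetic, following the attached parent mesh and adding secondary motion.
# It copies the parent pose, switching to a layered neck/head blend for skydiving.
# A neck-relative RigidBody simulation then feeds four chained, partially weighted AnimDynamics nodes on the dyn_hair bones.
# The cached result optionally passes through a bone-driven controller that adjusts ponytail_01 from left clavicle X translation, gated by bUsePonytailAdjustment.
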AnimBlueprint /BRCosmetics/Accessories/Hats/F_MED_Winter_Holiday_Hat/Meshes/F_MED_Winter_Holiday_Skeleton_AnimBP
target_skeleton=&/F_MED_Winter_Holiday_Skeleton
graphs [
 AnimGraph {nodes [
   n1 = "ponytail_01.multiple = clavicle_l.translateX * -4 - Bone Driven Controller"!(ComponentPose:Component Space Pose Link Structure=Local To Component(LocalPose=Use cached pose 'CachePose'()), Alpha:float=1) -> (Pose:Component Space Pose Link Structure) {
    is=title AnimGraphNode_BoneDrivenController
    bAffectTargetRotationX=true
    bAffectTargetRotationY=true
    bAffectTargetRotationZ=true
    bAffectTargetTranslationX=true
    bAffectTargetTranslationY=true
    bAffectTargetTranslationZ=true
    multiplier=-4
    sourceComponent=TranslationX
    alphaBoolBlend {blendOption=Linear}
    sourceBone {boneName=clavicle_l}
    targetBone {boneName=ponytail_01}
   }
   Output Pose(Result=Blend Poses by bool(
     BlendPose_0=Component To Local(ComponentPose=@n1)
     BlendPose_1=Use cached pose 'CachePose'()
     bActiveValue=Get bUsePonytailAdjustment() -> (bUsePonytailAdjustment:bool)
    ) {blendType=Linear}) {layerGroup=None}
   Save cached pose 'HeadLockCache'(Pose=Copy Pose From Mesh() {bCopyCurves=true, bUseAttachedParent=true}) {cacheName=HeadLockCache}
   n2 = Copy Pose From Mesh() {bUseAttachedParent=true}
   n3 = Layered blend per bone!(
    BasePose=Use cached pose 'HeadLockCache'()
    BlendPoses_0
    BlendPoses_1:Pose Link Structure=Use cached pose 'HeadLockCache'()
    BlendWeights_0=0.14
    BlendWeights_1:float=1
   ) -> (Pose) {
    blendPoses [{linkId=-1, sourceLinkId=-1}, {linkId=-1, sourceLinkId=-1}]
    blendWeights [1,1]
    layerSetup [{branchFilters [{blendDepth=1, boneName=neck_01}, {blendDepth=1, boneName=head}]}, {branchFilters []}]
   }
   n4 = Get bIsSkyDiving()
   n5 = RigidBody(ComponentPose=Local To Component(LocalPose=Blend Poses by bool(BlendPose_0=@n3, BlendPose_1=@n2, bActiveValue=@n4) {blendType=Linear})) {
    bClampLinearTranslationLimitToRefPose=true
    bOverrideWorldGravity=true
    overridePhysicsAsset=&/F_MED_Winter_Holiday_Physics
    simulationSpace=BaseBoneSpace
    alphaBoolBlend {blendOption=Linear}
    baseBoneRef {boneName=neck_01}
    componentAppliedLinearAccClamp=<3000 3000 3000>
    componentLinearAccScale=<0.32 0.32 0.5>
    componentLinearVelScale=<0.3 0.3 0.5>
    overrideWorldGravity {z=-980}
   }
   n6 = Anim Dynamics - Bone: dyn_hair_01(ComponentPose=@n5, Alpha=0.33) {
    bGravityOverrideInSimSpace=true
    alphaBoolBlend {blendOption=Linear}
    boundBone {boneName=dyn_hair_01}
    physicsBodyDefinitions [{
      sphereCollisionRadius=0
      boundBone {boneName=dyn_hair_01}
      boxExtents=<20 20 20>
      constraintSetup {angularLimitsMax=<0 61.1999 61.1999>, angularLimitsMin=<0 -57.5999 -43.2>, angularTarget=<0 0 0>}
      localJointOffset=<3 0 0>
     }]
   }
   n7 = Anim Dynamics - Bone: dyn_hair_02(ComponentPose=@n6, Alpha=0.33) {
    bGravityOverrideInSimSpace=true
    alphaBoolBlend {blendOption=Linear}
    boundBone {boneName=dyn_hair_02}
    physicsBodyDefinitions [{
      sphereCollisionRadius=0
      boundBone {boneName=dyn_hair_02}
      boxExtents=<20 20 20>
      constraintSetup {angularLimitsMax=<0 61.1999 61.1999>, angularLimitsMin=<0 -57.5999 -43.2>, angularTarget=<0 0 0>}
      localJointOffset=<3 0 0>
     }]
   }
   n8 = Anim Dynamics - Bone: dyn_hair_03(ComponentPose=@n7, Alpha=0.1, ExternalForce="0,0,0") {
    bGravityOverrideInSimSpace=true
    alphaBoolBlend {blendOption=Linear}
    boundBone {boneName=dyn_hair_03}
    physicsBodyDefinitions [{
      sphereCollisionRadius=0
      boundBone {boneName=dyn_hair_03}
      boxExtents=<20 20 20>
      constraintSetup {angularLimitsMax=<0 61.1999 61.1999>, angularLimitsMin=<0 -57.5999 -43.2>, angularTarget=<0 0 0>}
      localJointOffset=<3 0 0>
     }]
   }
   n9 = Anim Dynamics - Bone: dyn_hair_04(ComponentPose=@n8, Alpha=.33, ExternalForce="0,0,0") {
    bGravityOverrideInSimSpace=true
    alphaBoolBlend {blendOption=Linear}
    boundBone {boneName=dyn_hair_04}
    physicsBodyDefinitions [{
      sphereCollisionRadius=0
      boundBone {boneName=dyn_hair_04}
      boxExtents=<20 20 20>
      constraintSetup {angularLimitsMax=<0 0 0>, angularLimitsMin=<0 0 0>, angularTarget=<0 0 0>, linearAxesMax=<0.5 0.5 0.5>, linearAxesMin=<-0.5 -0.5 -0.5>}
      localJointOffset=<3 0 0>
     }]
   }
   Save cached pose 'CachePose'(Pose=Component To Local(ComponentPose=@n9)) {cacheName=CachePose}
  ]}
 EventGraph {links [], nodes []}
]
variables [test {type=boolean}]
